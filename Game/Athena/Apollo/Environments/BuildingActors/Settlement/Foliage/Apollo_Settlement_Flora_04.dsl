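; Construction script for Apollo_Settlement_Flora_04, a settlement foliage building actor that keeps a looping plant Niagara effect matched to its static mesh.
; After the parent script runs, it copies the mesh component's world location, rotation and scale onto the NS_Settlement_Plant_Loop component.
; It then passes the static mesh and world scale into the system as the PlantStaticMesh and PlantWorldScale parameters.
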(fn ConstructionScript
  (bind _ns_settlement_plant_loop (Variables|Default|GetNS_Settlement_Plant_Loop))
  (bind _staticmeshcomponent (Variables|Debug|GetStaticMeshComponent))
  (|CallParentFunction)
  (Transformation|SetWorldLocationAndRotation _ns_settlement_plant_loop (Transformation|GetWorldLocation _staticmeshcomponent) (Transformation|GetWorldRotation _staticmeshcomponent) false false)
  (Transformation|SetWorldScale3D _ns_settlement_plant_loop (Transformation|GetWorldScale _staticmeshcomponent))
  (Niagara|SetNiagaraStaticMeshDirectly _ns_settlement_plant_loop "PlantStaticMesh" (Variables|/Script/Engine.StaticMeshComponent|GetStaticMesh _staticmeshcomponent))
  (Niagara|SetNiagaraVariable(Vector3) _ns_settlement_plant_loop "PlantWorldScale" (Transformation|GetWorldScale (Variables|Debug|GetStaticMeshComponent))))
